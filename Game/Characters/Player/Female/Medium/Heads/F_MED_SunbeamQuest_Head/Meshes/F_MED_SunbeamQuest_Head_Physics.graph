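# Physics asset for the SunbeamQuest female medium player head, defining capsule bodies for a back-hair bone chain and the upper body around it.
# The hair runs from a tiny parent capsule on neck_02 through a base and four segments, with tight swing and twist limits that loosen in swing2 toward the tip.
# Upper spine, neck and clavicle capsules cover the surrounding body, joined by angularly free constraints, as is the hair parent's link to the neck.
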PhysicsAsset /Game/Characters/Player/Female/Medium/Heads/F_MED_SunbeamQuest_Head/Meshes/F_MED_SunbeamQuest_Head_Physics
bodies {
 clavicle_l [{
   shapeType=Capsule
   length=22.7388
   radius=8.9048
   shapeName=clavicle_l_capsule
   center=<5.2202 5.212 2.7664>
   rotation=<rot 90 180 -168.9964>
  }]
 clavicle_r [{
   shapeType=Capsule
   length=22.7388
   radius=8.9048
   shapeName=clavicle_l_capsule
   center=<-5.2202 -5.212 -2.7664>
   rotation=<rot 90 0 -168.9964>
  }]
 dyn_hair_back_01 [{shapeType=Capsule, length=5.0729, radius=2.5124, shapeName=dyn_hair_back_01_capsule, center=<4.0256 0 0>, rotation=<rot -90 0 0>}]
 dyn_hair_back_02 [{shapeType=Capsule, length=4.1786, radius=2.2846, shapeName=dyn_hair_back_02_capsule, center=<3.7757 0 0>, rotation=<rot -90 0 0>}]
 dyn_hair_back_03 [{shapeType=Capsule, length=3.233, radius=3.9608, shapeName=dyn_hair_back_03_capsule, center=<4.3191 0 0>, rotation=<rot -90 180 180>}]
 dyn_hair_back_04 [{shapeType=Capsule, length=3.9793, radius=2.8245, shapeName=dyn_hair_back_04_capsule, center=<4.1304 0 0>, rotation=<rot -90 0 0>}]
 dyn_hair_base [{shapeType=Capsule, length=5.951, radius=2.8971, shapeName=dyn_hair_base_capsule, center=<4.6624 0 0>, rotation=<rot -90 0 0>}]
 dyn_hair_parent [{shapeType=Capsule, length=0.5, radius=0.5, shapeName=dyn_hair_parent_capsule, rotation=<rot -90 0 0>}]
 neck_01 [{shapeType=Capsule, length=3.5959, radius=4.1963, shapeName=neck_01_capsule, center=<0.0258 0.4394 0>, rotation=<rot -90 -178.2101 170.5132>}]
 neck_02 [{shapeType=Capsule, length=3.9062, radius=4.0924, shapeName=neck_02_capsule, center=<1.779 0 0>, rotation=<rot -90 180 180>}]
 spine_04 [{shapeType=Capsule, length=15.4078, radius=13.477, shapeName=clavicle_l_capsule, center=<7.9234 3.3957 0>}]
 spine_05 [{shapeType=Capsule, length=18.8375, radius=3.9593, shapeName=spine_05_capsule, center=<16.194 -1.8084 0>}]
}
constraints [
 {bone1Name=dyn_hair_base, bone2Name=dyn_hair_parent, swing1LimitDegrees=10, swing2LimitDegrees=4, twistLimitDegrees=2}
 {bone1Name=dyn_hair_back_01, bone2Name=dyn_hair_base, swing1LimitDegrees=10, swing2LimitDegrees=2, twistLimitDegrees=2}
 {bone1Name=dyn_hair_back_02, bone2Name=dyn_hair_back_01, swing1LimitDegrees=10, swing2LimitDegrees=2, twistLimitDegrees=2}
 {bone1Name=dyn_hair_back_03, bone2Name=dyn_hair_back_02, swing1LimitDegrees=10, swing2LimitDegrees=6, twistLimitDegrees=2}
 {bone1Name=dyn_hair_back_04, bone2Name=dyn_hair_back_03, swing1LimitDegrees=10, swing2LimitDegrees=10, twistLimitDegrees=2}
 {
  bone1Name=dyn_hair_parent
  bone2Name=neck_02
  swing1LimitDegrees=45
  swing1Motion=Free
  swing2LimitDegrees=45
  swing2Motion=Free
  twistLimitDegrees=45
  twistMotion=Free
 }
 {
  bone1Name=neck_01
  bone2Name=spine_05
  swing1LimitDegrees=45
  swing1Motion=Free
  swing2LimitDegrees=45
  swing2Motion=Free
  twistLimitDegrees=45
  twistMotion=Free
 }
 {
  bone1Name=neck_02
  bone2Name=neck_01
  swing1LimitDegrees=45
  swing1Motion=Free
  swing2LimitDegrees=45
  swing2Motion=Free
  twistLimitDegrees=45
  twistMotion=Free
 }
 {
  bone1Name=spine_05
  bone2Name=spine_04
  swing1LimitDegrees=45
  swing1Motion=Free
  swing2LimitDegrees=45
  swing2Motion=Free
  twistLimitDegrees=45
  twistMotion=Free
 }
 {
  bone1Name=clavicle_l
  bone2Name=spine_05
  swing1LimitDegrees=45
  swing1Motion=Free
  swing2LimitDegrees=45
  swing2Motion=Free
  twistLimitDegrees=45
  twistMotion=Free
 }
 {
  bone1Name=clavicle_r
  bone2Name=spine_05
  swing1LimitDegrees=45
  swing1Motion=Free
  swing2LimitDegrees=45
  swing2Motion=Free
  twistLimitDegrees=45
  twistMotion=Free
 }
]
